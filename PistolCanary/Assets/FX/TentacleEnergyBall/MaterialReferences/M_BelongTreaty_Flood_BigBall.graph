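# Translucent unlit material for Niagara mesh particles that draws the big purple energy ball in the tentacle energy ball effect.
# Panned smoke, noise and distorted nebula textures combine with a Fresnel-driven curve atlas gradient and a bright stair-texture layer.
# The summed color is tinted by particle color and passed through MF_FX_EmissiveAdjustments.
# Opacity combines Fresnel masks, a Use Center Mask blend, depth fade and particle alpha.
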Material /PistolCanary/Assets/FX/TentacleEnergyBall/MaterialReferences/M_BelongTreaty_Flood_BigBall
graph {
 properties {
  bUsedWithNiagaraMeshParticles=true
  blendMode=BLEND_Translucent
  pixelDepthOffsetMode=PDOM_Legacy
  shadingModel=MSM_Unlit
  usedShadingModels=Unlit
 }
 nodes [
  n1 = Panner(Coordinate=TextureCoordinate() {uTiling=3, vTiling=2}) {speedX=-1}
  n2 = Multiply(A=Desaturation(
    None=TextureSample(UVs=@n1) {texture=/Game/Gadgets/Assets/VinderTech_GliderChute/Glider_Cube_Queen_Female/FX/T_CubeQueen_Glider_Stair}.RGB
    Fraction=Constant() {r=0.9}
   )) {constB=3}
  n3 = TextureSample() {samplerType=SAMPLERTYPE_Grayscale, texture=/Game/Global/Textures/Default/Noise/TilingNoise06}
  n4 = TextureCoordinate() {vTiling=1.2}
  n5 = ComponentMask(None=@n4) {g=true}
  n6 = AppendVector(A=ComponentMask(None=@n4) {r=true}, B=Power(Base=LinearInterpolate(Alpha=@n5)) {constExponent=1})
  n7 = Constant()
  n8 = TextureSample(UVs=Panner(Coordinate=TextureCoordinate() {vTiling=3}) {speedX=0.3}) {samplerType=SAMPLERTYPE_Grayscale, texture=&LowResBlurredNoise}
  n9 = ComponentMask(None=TextureCoordinate() {uTiling=2}) {g=true}
  n10 = Saturate(None=Multiply(A=AppendVector(A=@n7, B=@n8.R), B=Power(Base=@n9)))
  n11 = Constant() {r=-0.1}
  n12 = Constant()
  n13 = TextureSample(UVs=Panner(Coordinate=Add(A=@n6, B=@n10) {constB=0}, Speed=AppendVector(A=@n11, B=@n12)) {speedX=-0.2, speedY=-0.5}) {texture=&T_Inky_Smoke_Tile}
  n14 = TextureCoordinate() {uTiling=2}
  n15 = ComponentMask(None=@n14) {g=true}
  n16 = AppendVector(A=ComponentMask(None=@n14) {r=true}, B=Power(Base=LinearInterpolate(Alpha=@n15) {constB=3}))
  n17 = Constant() {r=-0.3}
  n18 = Constant() {r=-0.5}
  n19 = TextureSample(UVs=Panner(Coordinate=Add(A=@n16) {constB=0}, Speed=AppendVector(A=@n17, B=@n18)) {speedX=-0.2, speedY=-0.5}) {samplerType=SAMPLERTYPE_Grayscale, texture=&LowResBlurredNoise}
  n20 = TextureSample() {texture=/Game/Effects/Fort_Effects/Textures/Tile/T_Noise_02}
  n21 = LinearInterpolate(
   A=Constant3Vector() {constant=<rgba 0.0084 0 0.0417 1>}
   B=Constant3Vector() {constant=<rgba 0.3971 0.0335 1 1>}
   Alpha=Power(Base=LinearInterpolate(A=@n13.R, B=@n19.R, Alpha=@n20.R))
  )
  n22 = Constant3Vector() {constant=<rgba 0.313 0.11 1 1>}
  n23 = TextureCoordinate() {uTiling=2, vTiling=2}
  n24 = TextureCoordinate() {uTiling=2}
  n25 = Multiply(A=TextureSample(UVs=Panner(Coordinate=@n24) {speedX=-0.2}) {texture=&T_Noise_02}.R) {constB=2}
  n26 = TextureSample(UVs=Panner(Coordinate=Add(A=@n23, B=@n25)) {speedX=-0.1, speedY=-0.1}) {texture=/Game/Characters/Player/Male/Medium/Bodies/M_Med_Soldier_04/Skins/Galaxy/Textures/T_FN_Nebula_Pink}
  n27 = Fresnel() {baseReflectFraction=0, exponent=1}
  n28 = Multiply(A=Power(Base=Multiply(A=@n27)) {constExponent=1}) {constB=10}
  n29 = LinearInterpolate(
   A=TextureSample() {texture=/Game/Gadgets/Assets/VinderTech_GliderChute/Glider_Cube_Queen_Female/FX/T_Cheap_blue_space_color}.RGB
   B=Multiply(A=Multiply(A=@n26.RGB) {constB=2}, B=Saturate(None=@n28))
  ) {constAlpha=0.2}
  n30 = OneMinus(None=Multiply(A=Fresnel() {baseReflectFraction=0, exponent=1}))
  n31 = CurveAtlasRowParameter(CurveTime=Multiply(A=Power(Base=@n30) {constExponent=5}) {constB=0.2}) {
   atlas=&/Gameplay_BelongTreaty_Flood
   curve=&/Curve_BelongTreaty_Flood
   parameterName=Color Curve
  }
  n32 = Multiply(A=@n31, B=@n29)
  n33 = Constant3Vector() {constant=<rgba 0.9533 0.1848 1 1>}
  n34 = Panner(Coordinate=TextureCoordinate() {vTiling=2}) {speedY=-0.5}
  n35 = TextureCoordinate() {vTiling=2}
  n36 = Multiply(A=TextureSample(UVs=Panner(Coordinate=@n35) {speedX=-0.1}) {texture=&T_Inky_Smoke_Tile}.R) {constB=0.2}
  n37 = Power(Base=TextureSample(UVs=Add(A=@n34, B=@n36)) {texture=&T_Inky_Smoke_Tile}.R)
  n38 = Multiply(A=Multiply(A=LinearInterpolate(B=@n33, Alpha=@n37))) {constB=0}
  n39 = Fresnel() {baseReflectFraction=0, exponent=1}
  n40 = Power(Base=OneMinus(None=Multiply(A=@n39))) {constExponent=8}
  n41 = Multiply(A=@n40, B=@n40)
  n42 = Add(A=Add(A=Multiply(A=@n22, B=@n29), B=Add(B=@n32)), B=Add(B=Multiply(A=@n38, B=@n41)))
  n43 = Add(
   A=Multiply(A=LinearInterpolate(B=@n2, Alpha=@n3.R), B=Constant3Vector() {constant=<rgba 5.268 0 7 1>})
   B=Multiply(A=Add(A=@n21, B=@n42)) {constB=2}
  )
  n44 = ParticleColor()
  MaterialFunctionCall(Emissive=Add(B=Multiply(A=@n43, B=@n44.RGB))) {materialFunction=&MF_FX_EmissiveAdjustments}
  n45 = Fresnel() {baseReflectFraction=0, exponent=10}
  n46 = Power(Base=OneMinus(None=Multiply(A=@n45))) {constExponent=8}
  n47 = Fresnel() {baseReflectFraction=0, exponent=1}
  n48 = Power(Base=OneMinus(None=Multiply(A=@n47)))
  n49 = Fresnel() {baseReflectFraction=0, exponent=1}
  n50 = Multiply(
   A=Multiply(A=Saturate(None=@n46), B=LinearInterpolate(Alpha=@n48) {constA=0.5}) {constA=1}
   B=LinearInterpolate(B=Power(Base=@n49) {constExponent=1}, Alpha=ScalarParameter() {parameterName=Use Center Mask}) {constA=1}
  )
  Multiply(A=@n44.A, B=DepthFade(Opacity=Add(A=@n50) {constB=0}, FadeDistance=ScalarParameter() {defaultValue=5, parameterName=FadeDistance}))
 ]
}
