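# Unlit translucent material for Niagara mesh particles that renders a glowing lightning strand, filed under the TentacleEnergyBall effect.
# SplineThicken sets strand width from a sine profile along V, and time-animated sine offsets on local X and Y add wobble, both scaled by dynamic parameters.
# A panning lightning texture with per-particle randomized V tiling, masked and multiplied by particle alpha, forms opacity; color times particle color feeds emissive.
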Material /PistolCanary/Assets/FX/TentacleEnergyBall/MaterialReferences/M_BelongTreaty_Flood_Lightning
graph {
 properties {
  bUsedWithNiagaraMeshParticles=true
  blendMode=BLEND_Translucent
  pixelDepthOffsetMode=PDOM_Legacy
  shadingModel=MSM_Unlit
  usedShadingModels=Unlit
 }
 nodes [
  n1 = ComponentMask(None=TextureCoordinate()) {g=true}
  n2 = Multiply(A=Power(Base=Sine(None=@n1) {period=2})) {constB=10}
  n3 = Multiply(
   A=Multiply(A=Constant() {r=20}, B=DynamicParameter() {defaultValue=<rgba 1 1 1 0>, paramNames [WidthBase, V Scale, WPO, offset]}.WidthBase)
   B=LinearInterpolate(Alpha=Saturate(None=@n2)) {constA=0.1}
  )
  n4 = Multiply(A=TextureCoordinate() {vTiling=2}, B=AppendVector(A=Constant() {r=1}, B=ScalarParameter() {defaultValue=1, parameterName=WPO V Scale}))
  n5 = ScalarParameter() {defaultValue=-3, parameterName=WPO Speed}
  n6 = Add(A=ComponentMask(None=@n4) {g=true}, B=Multiply(A=Time(), B=@n5) {constB=2})
  n7 = Transform(None=Multiply(A=Sine(None=@n6), B=Constant3Vector() {constant=<rgba 1 0 0 0>})) {transformSourceType=TRANSFORMSOURCE_Local}
  n8 = DynamicParameter() {defaultValue=<rgba 1 1 1 0>, paramNames [WidthBase, V Scale, WPO, offset]}
  n9 = Add(A=ComponentMask(None=@n4) {g=true}, B=Multiply(A=Time(), B=@n5) {constB=2})
  n10 = Transform(None=Multiply(A=Sine(None=@n9), B=Constant3Vector() {constant=<rgba 0 1 0 0>})) {transformSourceType=TRANSFORMSOURCE_Local}
  Add(
   A=MaterialFunctionCall(WidthBase=@n3, WidthTip=@n3) {materialFunction=&SplineThicken}.WorldPositionOffset
   B=Add(A=Multiply(A=@n7, B=@n8.WPO), B=Multiply(A=@n10, B=@n8.WPO))
  )
  n11 = VectorParameter() {parameterName=Color, defaultValue=<rgba 8.8795 6.7708 10 1>}
  n12 = ParticleColor()
  MaterialFunctionCall(Emissive=Multiply(A=Add(B=@n11.RGB), B=@n12.RGB)) {materialFunction=&MF_FX_EmissiveAdjustments}
  n13 = Constant() {r=1}
  n14 = ParticleRandom()
  n15 = Add(
   A=Multiply(
    A=LinearInterpolate(Alpha=@n14) {constA=0.5, constB=0.8}
    B=DynamicParameter() {defaultValue=<rgba 1 1 1 0>, paramNames [WidthBase, V Scale, WPO, offset]}.V Scale
   )
   B=ParticleRandom()
  )
  n16 = Panner(Coordinate=Multiply(A=TextureCoordinate() {vTiling=2}, B=AppendVector(A=@n13, B=@n15))) {speedY=2}
  Multiply(
   A=StaticComponentMaskParameter(None=TextureSampleParameter2D(UVs=@n16) {parameterName=Defalut_Tex, texture=&/T_BelongTreaty_Flood_Lightning}.RGB) {defaultR=true, parameterName=Mask}
   B=@n12.A
  )
 ]
}
